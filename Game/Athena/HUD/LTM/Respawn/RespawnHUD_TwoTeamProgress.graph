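# Widget Blueprint for one team's score bar in the Respawn LTM two-team HUD, showing score text over a progress fill.
# PreConstruct uses isEnemy to pick ally or enemy materials and colors, mirror the backplate and set the scaling pivot.
# OnUpdateScore sets text and the dynamic material's PercentFill from team data, while UpdateScore derives fill from GoalScore and plays ScoreUp.
# ScoreThreshold selects pop animations by amount, adding an endlessly looping pulse for amount 2.
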WidgetBlueprint /Game/Athena/HUD/LTM/Respawn/RespawnHUD_TwoTeamProgress
animations [
 ScoreUp {
  binding_count=2
  display_label=""
  length_seconds=0.5824
  object=&self:ScoreUp
  sequence {
   bindings [
    ScoreTextBox {
     id=#1
     possessed_class=CommonTextBlock
     tracks [{
       type=MovieScene2DTransformTrack
       sections [{
         type=MovieScene2DTransformSection
         end_frame=34
         start_frame=0
         channels [
          Translation.X {type=MovieSceneScriptingFloatChannel}
          Translation.Y {type=MovieSceneScriptingFloatChannel}
          Angle {type=MovieSceneScriptingFloatChannel}
          Scale.X {type=MovieSceneScriptingFloatChannel, keys [[0,1],[2.016,1.131],[15.424,1]]}
          Scale.Y {type=MovieSceneScriptingFloatChannel, keys [[0,1],[2.016,1.131],[15.424,1]]}
          Shear.X {type=MovieSceneScriptingFloatChannel}
          Shear.Y {type=MovieSceneScriptingFloatChannel}
         ]
        }]
      }]
    }
    Image_BackplateProgress {
     id=#2
     possessed_class=Image
     tracks [{
       type=MovieSceneWidgetMaterialTrack
       sections [{
         type=MovieSceneParameterSection
         end_frame=34
         start_frame=0
         channels [ShineOffset {type=MovieSceneScriptingFloatChannel, keys [[0,0.64],[34.646,-1.1953]]}]
        }]
      }]
    }
   ]
   display_rate {numerator=60}
  }
 }
 ScoreThreshold1 {
  binding_count=1
  display_label=""
  length_seconds=0.2595
  object=&self:ScoreThreshold1
  sequence {
   bindings [HorizontalBox_Container {
     id=#3
     possessed_class=HorizontalBox
     tracks [{
       type=MovieScene2DTransformTrack
       sections [{
         type=MovieScene2DTransformSection
         end_frame=15
         start_frame=0
         channels [
          Translation.X {type=MovieSceneScriptingFloatChannel}
          Translation.Y {type=MovieSceneScriptingFloatChannel}
          Angle {type=MovieSceneScriptingFloatChannel}
          Scale.X {type=MovieSceneScriptingFloatChannel, keys [[0,1],[4,1.2],[15.57,1]]}
          Scale.Y {type=MovieSceneScriptingFloatChannel, keys [[0,1],[4,1.2],[15.57,1]]}
          Shear.X {type=MovieSceneScriptingFloatChannel}
          Shear.Y {type=MovieSceneScriptingFloatChannel}
         ]
        }]
      }]
    }]
   display_rate {numerator=60}
  }
 }
 ScoreThreshold2 {
  binding_count=1
  display_label=""
  length_seconds=0.5019
  object=&self:ScoreThreshold2
  sequence {bindings [HorizontalBox_Container {
     id=#4
     possessed_class=HorizontalBox
     tracks [{
       type=MovieScene2DTransformTrack
       sections [{
         type=MovieScene2DTransformSection
         end_frame=10
         start_frame=0
         channels [
          Translation.X {type=MovieSceneScriptingFloatChannel}
          Translation.Y {type=MovieSceneScriptingFloatChannel}
          Angle {type=MovieSceneScriptingFloatChannel}
          Scale.X {type=MovieSceneScriptingFloatChannel, keys [[0,1],[1,1.4008],[2,1.3314],[3,1.4043],[10.0367,1]]}
          Scale.Y {type=MovieSceneScriptingFloatChannel, keys [[0,1],[1,1.4008],[2,1.3314],[3,1.4043],[10.0367,1]]}
          Shear.X {type=MovieSceneScriptingFloatChannel}
          Shear.Y {type=MovieSceneScriptingFloatChannel}
         ]
        }]
      }]
    }]}
 }
 ScoreThreshold3 {
  binding_count=1
  display_label=""
  length_seconds=1.0639
  object=&self:ScoreThreshold3
  sequence {bindings [HorizontalBox_Container {
     id=#5
     possessed_class=HorizontalBox
     tracks [{
       type=MovieScene2DTransformTrack
       sections [{
         type=MovieScene2DTransformSection
         end_frame=21
         start_frame=0
         channels [
          Translation.X {type=MovieSceneScriptingFloatChannel}
          Translation.Y {type=MovieSceneScriptingFloatChannel}
          Angle {type=MovieSceneScriptingFloatChannel}
          Scale.X {type=MovieSceneScriptingFloatChannel, keys [[0,1],[2,1.4917],[4,1.3225],[6,1.4314],[10,1.2608],[12,1.3668],[21.277,1]]}
          Scale.Y {type=MovieSceneScriptingFloatChannel, keys [[0,1],[2,1.4917],[4,1.3225],[6,1.4314],[10,1.2608],[12,1.3668],[21.277,1]]}
          Shear.X {type=MovieSceneScriptingFloatChannel}
          Shear.Y {type=MovieSceneScriptingFloatChannel}
         ]
        }]
      }]
    }]}
 }
 PulseBar {
  binding_count=1
  display_label=""
  length_seconds=0.2704
  object=&self:PulseBar
  sequence {bindings [Image_BackPlatePulse {
     id=#6
     possessed_class=Image
     tracks [{
       type=MovieSceneFloatTrack
       display_name=Render Opacity
       sections [{
         type=MovieSceneFloatSection
         end_frame=5
         start_frame=0
         channels [None {type=MovieSceneScriptingFloatChannel, keys [[0,1],[2.1227,0],[5.406,1]]}]
        }]
      }]
    }]}
 }
]
graphs [
 EventGraph {nodes [
   n1 = Get isEnemy() -> (isEnemy:bool)
   SetColorAndOpacity(
    self:Image Object Reference=[
     Get Image_TeamColorBar() -> (Image_TeamColorBar:Image Object Reference)
     Get Image_TeamColorBarEnemy() -> (Image_TeamColorBarEnemy:Image Object Reference)
    ]
    InColorAndOpacity:Linear Color Structure=Select(Option 0:Linear Color Structure=(R=0,G=0.2487,B=0.5208,A=1), Option 1:Linear Color Structure=(R=0.3646,G=0.0031,B=0,A=1), Index=@n1) -> (ReturnValue:Linear Color Structure)
   )
   > SetVisibility(
    self=Get Image_TeamColorBarEnemy()
    InVisibility=Select(Option 0:ESlateVisibility Enum=Collapsed, Option 1:ESlateVisibility Enum=HitTestInvisible, Index=@n1) -> (ReturnValue:ESlateVisibility Enum)
   )
   > n2 = SetVisibility(self=Get Image_TeamColorBar(), InVisibility=Select(Option 0=SelfHitTestInvisible, Option 1=Collapsed, Index=@n1))
   SetColorAndOpacity(
    execute=[@n2.then, @n15.then]
    self=Get Image_BackPlatePulse() -> (Image_BackPlatePulse:Image Object Reference)
    InColorAndOpacity=Select(Option 0:Linear Color Structure=(R=0,G=0.3916,B=1,A=1), Option 1:Linear Color Structure=(R=1,G=0,B=0.0091,A=1), Index=@n1) -> (ReturnValue:Linear Color Structure)
   )
   > SetRenderTransformPivot!(
    execute
    self:Widget Object Reference=Get HorizontalBox_Container() -> (HorizontalBox_Container:Horizontal Box Object Reference)
    Pivot:Vector 2D Structure=Select(Option 0:Vector 2D Structure=(X=0.8,Y=1), Option 1:Vector 2D Structure=(X=0.2,Y=1), Index=@n1) -> (ReturnValue:Vector 2D Structure)
   ) -> (then)
   n3 = OnUpdateScore!() -> (OutputDelegate, then, TeamScoreData:"TDMTeam Score Data Structure (by ref)") {is=title K2Node_CustomEvent}
   n4 = Break TDMTeam Score Data!(TDMTeamScoreData:"TDMTeam Score Data Structure (by ref)"=@n3.TeamScoreData) -> (CurrScoreText:Text, CurrScorePercent:float=~) {is=title K2Node_BreakStruct}
   > "SetText (Text)"(self=Get ScoreTextBox() -> (ScoreTextBox:Common Text Object Reference), InText=@n4)
   > n5 = GetDynamicMaterial(self=Get Image_BackplateProgress() -> (Image_BackplateProgress:Image Object Reference))
   > n6 = Set ProgressBarDMI(ProgressBarDMI:Material Instance Dynamic Object Reference=@n5) -> (Output_Get:Material Instance Dynamic Object Reference)
   > SetScalarParameterValue(self=@n6, ParameterName=PercentFill, Value=@n4.CurrScorePercent)
   > Set Timer by Event(Delegate=Create Event(), Time=1.2, bLooping=false)
   n7 = ScoreThreshold!() -> (OutputDelegate, then, Amount:int) {is=title K2Node_CustomEvent}
   > n8 = Switch on Int!(execute:Exec, Selection:int=@n7.Amount) -> (Default:Exec, 0:Exec, 1:Exec, 2:Exec)
   PlayAnimation(execute=@n8.0, InAnimation=Get ScoreThreshold2() -> (ScoreThreshold2:Widget Animation Object Reference))
   PlayAnimation(execute=@n8.1, InAnimation=Get ScoreThreshold3() -> (ScoreThreshold3:Widget Animation Object Reference))
   PlayAnimation(execute=@n8.2, InAnimation=Get ScoreThreshold3())
   > PlayAnimation(InAnimation=Get PulseBar() -> (PulseBar:Widget Animation Object Reference), NumLoopsToPlay=0)
   PlayAnimation(execute=@n8.Default, InAnimation=Get ScoreThreshold1() -> (ScoreThreshold1:Widget Animation Object Reference))
   PlayAnimationAndSound() {is=title K2Node_CustomEvent}
   > PlayAnimation(InAnimation=Get ScoreUp() -> (ScoreUp:Widget Animation Object Reference))
   Event PreConstruct()
   n9 = Get Image_BackplateProgress()
   > SetBrushFromMaterial(
    self=@n9
    Material=Select(
     Option 0:Material Interface Object Reference=/Game/Athena/HUD/Materials/Score/M_ScoreLTM_ScoreProgress_SquareAlly
     Option 1:Material Interface Object Reference=/Game/Athena/HUD/Materials/Score/M_ScoreLTM_ScoreProgress_SquareEnemy
     Index=@n1
    ) -> (ReturnValue:Material Interface Object Reference)
   )
   n10 = Get Image_BackPlatePulse()
   > SetBrushFromMaterial(
    self=@n10
    Material=Select(
     Option 0=/Game/Athena/HUD/Materials/Score/M_ScoreLTM_ScoreProgress_SquareWhite
     Option 1=/Game/Athena/HUD/Materials/Score/M_ScoreLTM_ScoreProgress_SquareWhite_Enemy
     Index=@n1
    )
   )
   > SetRenderScale!(
    execute
    self:Widget Object Reference=[@n10, @n9]
    Scale:Vector 2D Structure=Select(Option 0:Vector 2D Structure=(X=1,Y=1), Option 1:Vector 2D Structure=(X=-1,Y=1), Index=@n1) -> (ReturnValue:Vector 2D Structure)
   ) -> (then)
   n11 = Get ScoreTextBox()
   n12 = Break Slate Font Info!(SlateFontInfo:"Slate Font Info Structure (by ref)"=Get Font!(self:Text Object Reference=@n11) -> (Font:Slate Font Info Structure)) -> (
    FontObject:Object Reference
    FontMaterial:Object Reference
    OutlineSettings:Font Outline Settings Structure
    TypefaceFontName:Name
    Size:float=~
    LetterSpacing:int
    SkewAmount:float=~
    bForceMonospaced:bool=~
    bMaterialIsStencil:bool=~
    MonospacedWidth:float=~
   ) {is=title K2Node_BreakStruct}
   n13 = Break Font Outline Settings!(FontOutlineSettings:"Font Outline Settings Structure (by ref)"=@n12.OutlineSettings) -> (
    OutlineSize:int
    bMiteredCorners:bool=~
    bSeparateFillAlpha:bool=~
    bApplyOutlineToDropShadows:bool=~
    OutlineMaterial:Object Reference
    OutlineColor:Linear Color Structure
   ) {is=title K2Node_BreakStruct}
   n14 = Select(Option 0:Linear Color Structure=(R=0,G=0.2284,B=0.5208,A=1), Option 1:Linear Color Structure=(R=0.2865,G=0.0018,B=0,A=1), Index=@n1) -> (ReturnValue:Linear Color Structure)
   > n15 = SetFont(
    self=@n11
    InFontInfo=MakeSlateFontInfo!(
     FontObject:Object Reference=@n12
     FontMaterial:Object Reference=@n12.FontMaterial
     OutlineSettings:Font Outline Settings Structure=Make Font Outline Settings!(
      OutlineSize:int=@n13
      bMiteredCorners:bool=False
      bSeparateFillAlpha:bool=@n13.bSeparateFillAlpha
      bApplyOutlineToDropShadows:bool=@n13.bApplyOutlineToDropShadows
      OutlineMaterial:Object Reference=@n13.OutlineMaterial
      OutlineColor:Linear Color Structure=@n14
     ) -> (FontOutlineSettings:Font Outline Settings Structure) {is=title K2Node_MakeStruct}
     TypefaceFontName:Name=@n12.TypefaceFontName
     Size:float=@n12.Size
     LetterSpacing:int=0
     SkewAmount:float=0
     bForceMonospaced:bool=false
     bMaterialIsStencil:bool=false
     MonospacedWidth:float=1
    ) -> (ReturnValue:Slate Font Info Structure)
   )
   Event Construct()
   > "SetText (Text)"(self=Get ScoreTextBox())
   > UpdateScore!(execute, self, Score:int=0) -> (then)
  ]}
 UpdateScore {nodes [
   n1 = UpdateScore!() -> (then, Score:int) {is=title K2Node_FunctionEntry}
   > Branch(Condition="integer >= integer"!(A:int=@n1, B:int=0) -> (ReturnValue:bool) {is=title K2Node_CallFunction})
   > "SetText (Text)"(self=Get ScoreTextBox(), InText=Format Text!(Format="LOCTEXT(\"{points}\")", points:int=@n1) -> (Result))
   > Sequence()
   n2 = Get GoalScore() -> (GoalScore:int=0)
   > SetScalarParameterValue(
    self=Get ProgressBarDMI() -> (ProgressBarDMI:Material Instance Dynamic Object Reference)
    ParameterName=PercentFill
    Value=float / float(A="To Float (Integer)"(InInt=@n1), B="To Float (Integer)"(InInt=@n2)) {is=title K2Node_CallFunction}
   )
   > PlayAnimation(InAnimation=Get ScoreUp())
   > SetVisibility(InVisibility=HitTestInvisible)
  ]}
]
variables [
 GoalScore {type=integer, default=1000}
 isEnemy {type=boolean}
 ProgressBarDMI {type=MaterialInstanceDynamic}
]
widgets {
 info {inheritedWidgetCount=0, parentClass=&UserWidget, rootWidgetClass=&Overlay, widgetCount=9}
 widgets [
  Overlay_1 {widgetClassPath=&Overlay, widget=Overlay_1}
  SizeBox_0 {
   widgetClassPath=&SizeBox
   parent=Overlay_1
   widget=SizeBox_0
   properties {bOverride_MinDesiredWidth=true, heightOverride=51.1903, minDesiredWidth=81.7156, widthOverride=181.5583}
   slot {class=OverlaySlot, properties {horizontalAlignment=HAlign_Fill, verticalAlignment=VAlign_Fill}}
  }
  HorizontalBox_Container {
   widgetClassPath=&HorizontalBox
   bIsVariable=true
   parent=SizeBox_0
   widget=HorizontalBox_Container
   properties {renderTransformPivot=<0.8 1>}
   slot {class=SizeBoxSlot}
  }
  Image_TeamColorBarEnemy {
   widgetClassPath=&Image
   bIsVariable=true
   parent=HorizontalBox_Container
   widget=Image_TeamColorBarEnemy
   properties {
    renderOpacity=0
    visibility=Collapsed
    brush {
     bIsDynamicallyLoaded=false
     drawAs=Image
     imageType=NoImage
     mirroring=NoMirror
     resourceName=None
     resourceObject=null
     tiling=NoTile
     imageSize=<3 16>
     margin=<ltrb 0 0 0 0>
     outlineSettings {
      bUseBrushTransparency=false
      roundingType=HalfHeightRadius
      width=0
      color {colorUseRule=UseColor_Specified, specifiedColor=<rgba 0 0 0 0>}
      cornerRadii=<0 0 0 0>
     }
     tintColor {colorUseRule=UseColor_Specified, specifiedColor=<rgba 1 1 1 1>}
     uVRegion {bIsValid=false, max=<0 0>, min=<0 0>}
    }
    colorAndOpacity=<rgba 0 0.8308 1 1>
   }
   slot {class=HorizontalBoxSlot, properties {horizontalAlignment=HAlign_Right}}
  }
  Overlay_0 {
   widgetClassPath=&Overlay
   parent=HorizontalBox_Container
   widget=Overlay_0
   slot {class=HorizontalBoxSlot, properties {size {sizeRule=Fill, value=1}}}
  }
  Image_BackPlatePulse {
   widgetClassPath=&Image
   bIsVariable=true
   parent=Overlay_0
   widget=Image_BackPlatePulse
   properties {
    renderOpacity=0
    brush {
     bIsDynamicallyLoaded=false
     drawAs=Box
     imageType=NoImage
     mirroring=NoMirror
     resourceName=None
     resourceObject=&M_ScoreLTM_ScoreProgress_SquareWhite
     tiling=NoTile
     imageSize=<32 49.9545>
     margin=<ltrb 0.1 0.1 0.1 0.1>
     outlineSettings {
      bUseBrushTransparency=false
      roundingType=HalfHeightRadius
      width=0
      color {colorUseRule=UseColor_Specified, specifiedColor=<rgba 0 0 0 0>}
      cornerRadii=<0 0 0 0>
     }
     tintColor {colorUseRule=UseColor_Specified, specifiedColor=<rgba 1 1 1 1>}
     uVRegion {bIsValid=false, max=<0 0>, min=<0 0>}
    }
    colorAndOpacity=<rgba 0 0.3916 1 1>
    renderTransform {angle=0, scale=<1 1>, shear=<-9 0>, translation=<0 0>}
   }
   slot {class=OverlaySlot, properties {horizontalAlignment=HAlign_Fill, verticalAlignment=VAlign_Fill, padding=<ltrb 0 -8 0 -8>}}
  }
  Image_BackplateProgress {
   widgetClassPath=&Image
   bIsVariable=true
   parent=Overlay_0
   widget=Image_BackplateProgress
   properties {
    brush {
     bIsDynamicallyLoaded=false
     drawAs=Box
     imageType=NoImage
     mirroring=NoMirror
     resourceName=None
     resourceObject=&M_ScoreLTM_ScoreProgress_SquareAlly
     tiling=NoTile
     imageSize=<32 49.9545>
     margin=<ltrb 0.1 0.1 0.1 0.1>
     outlineSettings {
      bUseBrushTransparency=false
      roundingType=HalfHeightRadius
      width=0
      color {colorUseRule=UseColor_Specified, specifiedColor=<rgba 0 0 0 0>}
      cornerRadii=<0 0 0 0>
     }
     tintColor {colorUseRule=UseColor_Specified, specifiedColor=<rgba 1 1 1 1>}
     uVRegion {bIsValid=false, max=<0 0>, min=<0 0>}
    }
    renderTransform {angle=0, scale=<1 1>, shear=<-9 0>, translation=<0 0>}
   }
   slot {class=OverlaySlot, properties {horizontalAlignment=HAlign_Fill, verticalAlignment=VAlign_Fill, padding=<ltrb 0 -8 0 -8>}}
  }
  ScoreTextBox {
   widgetClassPath=&CommonTextBlock
   bIsVariable=true
   parent=Overlay_0
   widget=ScoreTextBox
   properties {
    text=0
    font {
     bForceMonospaced=false
     bMaterialIsStencil=false
     fontMaterial=null
     fontObject=&BurbankSmall
     letterSpacing=0
     monospacedWidth=1
     size=20
     skewAmount=0
     typefaceFontName=Bold
     outlineSettings {
      bApplyOutlineToDropShadows=false
      bMiteredCorners=false
      bSeparateFillAlpha=false
      outlineMaterial=null
      outlineSize=1
      outlineColor=<rgba 0 0.2284 0.5208 1>
     }
    }
    renderTransform {angle=0, scale=<1 1>, shear=<0 0>, translation=<1 1>}
    renderTransformPivot=<0.5 0.8>
    shadowColorAndOpacity=<rgba 0.0052 0.0052 0.0056 0.7619>
    shadowOffset=<3 2>
   }
   slot {class=OverlaySlot, properties {horizontalAlignment=HAlign_Center, verticalAlignment=VAlign_Center, padding=<ltrb 0 -4 0 -4>}}
  }
  Image_TeamColorBar {
   widgetClassPath=&Image
   bIsVariable=true
   parent=HorizontalBox_Container
   widget=Image_TeamColorBar
   properties {
    renderOpacity=0
    visibility=Collapsed
    brush {
     bIsDynamicallyLoaded=false
     drawAs=Image
     imageType=NoImage
     mirroring=NoMirror
     resourceName=None
     resourceObject=null
     tiling=NoTile
     imageSize=<3 16>
     margin=<ltrb 0 0 0 0>
     outlineSettings {
      bUseBrushTransparency=false
      roundingType=HalfHeightRadius
      width=0
      color {colorUseRule=UseColor_Specified, specifiedColor=<rgba 0 0 0 0>}
      cornerRadii=<0 0 0 0>
     }
     tintColor {colorUseRule=UseColor_Specified, specifiedColor=<rgba 1 1 1 1>}
     uVRegion {bIsValid=false, max=<0 0>, min=<0 0>}
    }
    colorAndOpacity=<rgba 0 0.8308 1 1>
   }
   slot {class=HorizontalBoxSlot, properties {horizontalAlignment=HAlign_Right}}
  }
 ]
}
