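# LTM_Disco_Countdown_Final_Cue is a sound cue for the final countdown in the Disco limited-time mode.
# A single wave player plays a shared LTM score in-game pop final-count sound at slightly reduced volume.
# It uses an important in-game mission sound class, a very large max distance, and a concurrency override that caps simultaneous instances at two.
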SoundCue /Game/Athena/Sounds/LTMs/Disco/LTM_Disco_Countdown_Final_Cue
root=WavePlayer_3
nodes [{id=WavePlayer_3, properties {soundWaveAssetPtr=/Game/Athena/Sounds/LTMs/Score/ScoreLTM_InGamePop_CountFinal_06}}]
properties {
 bOverrideConcurrency=true
 maxDistance=2097152
 soundClassObject=&Fort_Mission_Ingame_Important_PSM
 volumeMultiplier=0.9
 concurrencyOverrides {maxCount=2}
}
