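# Sound cue for the Disco LTM's intensity 1 enemy scoring sting, apparently a small negative in-game pop.
# A random node picks one of three equally weighted LittleBad pop waves.
# It plays at reduced volume on the enemy-damaged sound class, overrides concurrency to cap simultaneous instances at two, and has a very large max distance.
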SoundCue /Game/Athena/Sounds/LTMs/Disco/LTM_Disco_Intensity1_Enemy_Cue
root=Random_0
nodes [
 {id=Random_0, children [WavePlayer_0,WavePlayer_1,WavePlayer_2], properties {weights [1,1,1]}}
 {id=WavePlayer_0, properties {soundWaveAssetPtr=&/ScoreLTM_InGamePop_LittleBad_10}}
 {id=WavePlayer_1, properties {soundWaveAssetPtr=&/ScoreLTM_InGamePop_LittleBad_11}}
 {id=WavePlayer_2, properties {soundWaveAssetPtr=&/ScoreLTM_InGamePop_LittleBad_09}}
]
properties {
 bOverrideConcurrency=true
 maxDistance=2097152
 soundClassObject=/Game/Sounds/FortSoundClassesAndModes/Fort_Enemy_Damaged_PSM
 volumeMultiplier=0.4
 concurrencyOverrides {maxCount=2}
}
