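# Sound cue for the Disco LTM in Athena, named as the intensity 3 enemy variant.
# A single wave player plays the ScoreLTM_InGamePop_BigBad_08 sound at reduced volume, with no randomization or mixing.
# It uses the Fort_Enemy_Damaged_PSM sound class, sets a very large max distance, and overrides concurrency to cap simultaneous instances at two.
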SoundCue /Game/Athena/Sounds/LTMs/Disco/LTM_Disco_Intensity3_Enemy_Cue
root=WavePlayer_3
nodes [{id=WavePlayer_3, properties {soundWaveAssetPtr=&/ScoreLTM_InGamePop_BigBad_08}}]
properties {
 bOverrideConcurrency=true
 maxDistance=2097152
 soundClassObject=/Game/Sounds/FortSoundClassesAndModes/Fort_Enemy_Damaged_PSM
 volumeMultiplier=0.4
 concurrencyOverrides {maxCount=2}
}
