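# Sound cue for the Disco limited-time mode, named as the friendly variant of intensity level 3.
# A single wave player plays a reused LTM score pop sound at reduced volume.
# It uses the enemy-damaged sound class and overrides concurrency to allow at most two instances, with a very large max distance.
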SoundCue /Game/Athena/Sounds/LTMs/Disco/LTM_Disco_Intensity3_Friendly_Cue
root=WavePlayer_0
nodes [{id=WavePlayer_0, properties {soundWaveAssetPtr=/Game/Athena/Sounds/LTMs/Score/ScoreLTM_InGamePop_Big_04}}]
properties {
 bOverrideConcurrency=true
 maxDistance=2097152
 soundClassObject=/Game/Sounds/FortSoundClassesAndModes/Fort_Enemy_Damaged_PSM
 volumeMultiplier=0.4
 concurrencyOverrides {maxCount=2}
}
