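# Two-team score HUD widget for an LTM respawn mode, with ally and enemy progress widgets and a hidden-until-updated total score.
# Parent-class events route team scores by index to the progress widgets and set the total text.
# Score-progress events play Disco LTM countdown or team-specific intensity sounds and call ScoreThreshold, though non-countdown enemy calls always pass zero.
# On construct, the border shifts upward unless using compact touch UI.
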WidgetBlueprint /Game/Athena/HUD/LTM/Respawn/RespawnHUD_TwoTeamScores
graphs [
 EventGraph {nodes [
   Event Construct()
   > UpdateWidgetTranslation()
   n1 = Event OnUpdateTotalScore!() -> (OutputDelegate, then, TotalScoreText:"Text (by ref)", Score:int=0) {is=title K2Node_Event}
   > SetVisibility(self=Get TotalsBox() -> (TotalsBox:Horizontal Box Object Reference), InVisibility=SelfHitTestInvisible)
   > "SetText (Text)"(self=Get ScoreTextBox() -> (ScoreTextBox:Common Text Object Reference), InText=@n1.TotalScoreText)
   n2 = Event OnPlayScoreProgressSound!() -> (OutputDelegate, then, ScoreProgressType:ETDMScoreProgressTypes Enum=None, bIsLocalTeam:bool) {is=title K2Node_Event}
   > n3 = Switch on ETDMScoreProgressTypes!(execute, Selection:ETDMScoreProgressTypes Enum=@n2.ScoreProgressType) -> (None:Exec, ProgressSoundMild:Exec, ProgressSoundMedium:Exec, ProgressSoundStrong:Exec, CountdownSoundNormal:Exec, CountdownSoundStrong:Exec) {is=title K2Node_SwitchEnum}
   n4 = PlaySound2D(execute=@n3.CountdownSoundNormal, Sound=/Game/Athena/Sounds/LTMs/Disco/LTM_Disco_Countdown_Cue)
   n5 = Branch(execute=[@n6.then, @n4.then], Condition=@n2.bIsLocalTeam)
   > ScoreThreshold!(
    execute
    self:Respawn HUD Two Team Progress Object Reference=Get Progress_Ally() -> (Progress_Ally:Respawn HUD Two Team Progress Object Reference)
    Amount:int=2
   ) -> (then)
   ScoreThreshold(execute=@n5.else, self=Get Progress_Enemy() -> (Progress_Enemy:Respawn HUD Two Team Progress Object Reference), Amount=2)
   n6 = PlaySound2D(execute=@n3.CountdownSoundStrong, Sound=/Game/Athena/Sounds/LTMs/Disco/LTM_Disco_Countdown_Final_Cue)
   n7 = Branch(execute=[@n3.ProgressSoundMedium, @n3.ProgressSoundMild, @n3.ProgressSoundStrong], Condition=@n2.bIsLocalTeam)
   n8 = Select!(
    None:Sound Base Object Reference
    ProgressSoundMild:Sound Base Object Reference=/Game/Athena/Sounds/LTMs/Disco/LTM_Disco_Intensity1_Friendly_Cue
    ProgressSoundMedium:Sound Base Object Reference=/Game/Athena/Sounds/LTMs/Disco/LTM_Disco_Intensity2_Friendly_Cue
    ProgressSoundStrong:Sound Base Object Reference=/Game/Athena/Sounds/LTMs/Disco/LTM_Disco_Intensity3_Friendly_Cue
    CountdownSoundNormal:Sound Base Object Reference
    CountdownSoundStrong:Sound Base Object Reference
    Index:ETDMScoreProgressTypes Enum=@n2.ScoreProgressType
   ) -> (ReturnValue:Sound Base Object Reference)
   > Is Valid(InputObject=@n8)
   > PlaySound2D(Sound=@n8)
   > ScoreThreshold(
    self=Get Progress_Ally()
    Amount=Select(
     None:int=0
     ProgressSoundMild:int=0
     ProgressSoundMedium:int=1
     ProgressSoundStrong:int=1
     CountdownSoundNormal:int=2
     CountdownSoundStrong:int=3
     Index=@n2.ScoreProgressType
    ) -> (ReturnValue:int)
   )
   n9 = Select(
    None:Sound Base Object Reference
    ProgressSoundMild:Sound Base Object Reference=/Game/Athena/Sounds/LTMs/Disco/LTM_Disco_Intensity1_Enemy_Cue
    ProgressSoundMedium:Sound Base Object Reference=/Game/Athena/Sounds/LTMs/Disco/LTM_Disco_Intensity2_Enemy_Cue
    ProgressSoundStrong:Sound Base Object Reference=/Game/Athena/Sounds/LTMs/Disco/LTM_Disco_Intensity3_Enemy_Cue
    CountdownSoundNormal:Sound Base Object Reference
    CountdownSoundStrong:Sound Base Object Reference
    Index=@n2.ScoreProgressType
   ) -> (ReturnValue:Sound Base Object Reference)
   Is Valid(exec=@n7.else, InputObject=@n9)
   > PlaySound2D(Sound=@n9)
   > ScoreThreshold(
    self=Get Progress_Enemy()
    Amount=Select!(Option 0:int=0, Option 1:int=0, Option 2:int=0, Option 3:int=0, Option 4:int=0, Option 5:int=0, Index:Byte=0) -> (ReturnValue:int)
   )
   n10 = Event OnUpdateTeamScore!() -> (OutputDelegate, then, TeamIndex:int=0, ScoreData:"TDMTeam Score Data Structure (by ref)") {is=title K2Node_Event}
   > OnUpdateScore!(
    execute
    self:Respawn HUD Two Team Progress Object Reference=Select!(
     Option 0:Respawn HUD Two Team Progress Object Reference=Get Progress_Ally()
     Option 1:Respawn HUD Two Team Progress Object Reference=Get Progress_Enemy()
     Index:int=@n10.TeamIndex
    ) -> (ReturnValue:Respawn HUD Two Team Progress Object Reference)
    TeamScoreData:"TDMTeam Score Data Structure (by ref)"=@n10.ScoreData
   ) -> (then)
  ]}
 UpdateWidgetTranslation {nodes [
   UpdateWidgetTranslation() {is=title K2Node_FunctionEntry}
   n1 = GetOwningPlayer()
   > n2 = Branch(Condition=AND Boolean(
     A=IsUsingCompactUI()
     B=IsUsingInputType!(InputType:ECommonInputType Enum=Touch, PlayerController:Player Controller Object Reference=@n1) -> (ReturnValue:bool)
    ))
   SetRenderTranslation!(execute=@n2.else, self=Get Border_0() -> (Border_0:Border Object Reference), Translation_X:double=0, Translation_Y:double=-40) -> (then)
  ]}
]
variables []
widgets {
 info {inheritedWidgetCount=1, parentClass=/Script/FortniteUI.TwoTeamScoreWidgetBase, rootWidgetClass=&Border, widgetCount=10}
 widgets [
  Border_0 {
   widgetClassPath=&Border
   bIsVariable=true
   widget=Border_0
   properties {
    background {
     bIsDynamicallyLoaded=false
     drawAs=NoDrawType
     imageType=NoImage
     mirroring=NoMirror
     resourceName=None
     resourceObject=null
     tiling=NoTile
     imageSize=<32 32>
     margin=<ltrb 0 0 0 0>
     outlineSettings {
      bUseBrushTransparency=false
      roundingType=HalfHeightRadius
      width=0
      color {colorUseRule=UseColor_Specified, specifiedColor=<rgba 0 0 0 0>}
      cornerRadii=<0 0 0 0>
     }
     tintColor {colorUseRule=UseColor_Specified, specifiedColor=<rgba 1 1 1 1>}
     uVRegion {bIsValid=false, max=<0 0>, min=<0 0>}
    }
    padding=<ltrb 15 0 15 0>
   }
  }
  VerticalBox_0 {
   widgetClassPath=&VerticalBox
   parent=Border_0
   widget=VerticalBox_0
   slot {class=BorderSlot, properties {padding=<ltrb 15 0 15 0>}}
  }
  Spacer_1 {widgetClassPath=&Spacer, parent=VerticalBox_0, widget=Spacer_1, properties {size=<258 10>}, slot {class=VerticalBoxSlot}}
  HorizontalBox_0 {
   widgetClassPath=&HorizontalBox
   parent=VerticalBox_0
   widget=HorizontalBox_0
   slot {class=VerticalBoxSlot, properties {size {sizeRule=Fill, value=1}}}
  }
  Progress_Ally {
   widgetClassPath=&/RespawnHUD_TwoTeamProgress
   bIsVariable=true
   parent=HorizontalBox_0
   widget=Progress_Ally
   slot {class=HorizontalBoxSlot, properties {verticalAlignment=VAlign_Center, size {sizeRule=Fill, value=1}}}
  }
  Progress_Enemy {
   widgetClassPath=&RespawnHUD_TwoTeamProgress
   bIsVariable=true
   parent=HorizontalBox_0
   widget=Progress_Enemy
   properties {isEnemy=true}
   slot {class=HorizontalBoxSlot, properties {verticalAlignment=VAlign_Center, size {sizeRule=Fill, value=1}}}
  }
  TotalsBox {
   widgetClassPath=&HorizontalBox
   bIsVariable=true
   parent=VerticalBox_0
   widget=TotalsBox
   properties {visibility=Hidden}
   slot {class=VerticalBoxSlot}
  }
  Overlay_1 {
   widgetClassPath=&Overlay
   parent=TotalsBox
   widget=Overlay_1
   slot {class=HorizontalBoxSlot, properties {size {sizeRule=Fill, value=1}}}
  }
  ScoreTextBox {
   widgetClassPath=&CommonTextBlock
   bIsVariable=true
   parent=Overlay_1
   widget=ScoreTextBox
   properties {
    justification=Center
    text=0
    font {
     bForceMonospaced=false
     bMaterialIsStencil=false
     fontMaterial=null
     fontObject=&BurbankSmall
     letterSpacing=0
     monospacedWidth=1
     size=16
     skewAmount=0
     typefaceFontName=Bold
     outlineSettings {
      bApplyOutlineToDropShadows=false
      bMiteredCorners=false
      bSeparateFillAlpha=false
      outlineMaterial=&M-FontShadow
      outlineSize=3
      outlineColor=<rgba 0 0 0 0.8>
     }
    }
    renderTransform {angle=0, scale=<1 1>, shear=<0 0>, translation=<1 1>}
    renderTransformPivot=<0.5 0.8>
    shadowColorAndOpacity=<rgba 0.0052 0.0052 0.0056 0.7619>
    shadowOffset=<0 0>
   }
   slot {class=OverlaySlot, properties {horizontalAlignment=HAlign_Fill, verticalAlignment=VAlign_Fill}}
  }
  TutorialHighlight {widgetClassPath=&TutorialHighlight}
 ]
}
