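# Widget Blueprint on TwoTeamScoreWidgetBase that relays team score text to game-over screen content, hosted in a named slot.
# Construct clears the TeamScore text array.
# OnUpdateTeamScore inserts CurrScoreText at the team index, rather than replacing the entry, then broadcasts the array through the Updated dispatcher.
# An Intro scale-and-fade animation targets HorizontalBox_0, absent from the listed widgets, and is never played in this graph.
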WidgetBlueprint /Game/Athena/HUD/LTM/Respawn/Customizable_GameOver_Score
animations [Intro {
  binding_count=1
  length_seconds=0.45
  object=&self:Intro
  sequence {bindings [HorizontalBox_0 {
     id=#1
     possessed_class=HorizontalBox
     tracks [
      {
       type=MovieScene2DTransformTrack
       sections [{
         type=MovieScene2DTransformSection
         channels [
          Translation.X {type=MovieSceneScriptingFloatChannel}
          Translation.Y {type=MovieSceneScriptingFloatChannel}
          Angle {type=MovieSceneScriptingFloatChannel}
          Scale.X {type=MovieSceneScriptingFloatChannel, keys [[0,1.2],[9,1]]}
          Scale.Y {type=MovieSceneScriptingFloatChannel, keys [[0,1.2],[9,1]]}
          Shear.X {type=MovieSceneScriptingFloatChannel}
          Shear.Y {type=MovieSceneScriptingFloatChannel}
         ]
        }]
      }
      {
       type=MovieSceneFloatTrack
       display_name=Render Opacity
       sections [{type=MovieSceneFloatSection, channels [None {type=MovieSceneScriptingFloatChannel, keys [[0,0],[9,1]]}]}]
      }
     ]
    }]}
 }]
graphs [
 EventGraph {nodes [
   Event Construct()
   > Clear(TargetArray:Array of Texts=Get TeamScore() -> (TeamScore:Array of Texts))
   n1 = Event OnUpdateTeamScore!() -> (OutputDelegate, then, TeamIndex:int=0, ScoreData:"TDMTeam Score Data Structure (by ref)") {is=title K2Node_Event}
   > Insert!(
    execute
    TargetArray:Array of Texts=Get TeamScore()
    NewItem:"Text (by ref)"=Break TDMTeam Score Data!(TDMTeamScoreData:"TDMTeam Score Data Structure (by ref)"=@n1.ScoreData) -> (CurrScoreText:Text, CurrScorePercent:float=~) {is=title K2Node_BreakStruct}
    Index:int=@n1.TeamIndex
   ) -> (then) {is=title K2Node_CallArrayFunction}
   > Call Updated!(execute:Exec, self:Customizable Game Over Score Object Reference, Scores:Array of Texts=Get TeamScore()) -> (then:Exec) {is=title K2Node_CallDelegate}
   Event PreConstruct()
  ]}
 Updated {nodes [Updated!() -> (then, Scores:Array of Texts) {is=title K2Node_FunctionEntry}]}
]
variables [
 TeamScore {type="array<string>", default []}
 Updated {}
]
widgets {
 info {inheritedWidgetCount=1, parentClass=/Script/FortniteUI.TwoTeamScoreWidgetBase, rootWidgetClass=&Overlay, widgetCount=3}
 widgets [
  Overlay_10 {widgetClassPath=&Overlay, widget=Overlay_10}
  WidgetSlot {
   widgetClassPath=&NamedSlot
   bIsVariable=true
   parent=Overlay_10
   widget=WidgetSlot
   properties {bExposeOnInstanceOnly=true}
   slot {class=OverlaySlot}
  }
  TutorialHighlight {widgetClassPath=&TutorialHighlight}
 ]
}
