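# RespawnPostgame_ScoreDisplay is a game-over score widget for respawn LTMs, showing a "FINAL SCORE" header above friendly and enemy team scores on gradient backgrounds.
# On construct it hides, binds to the parent's Updated event, then after a delay plays a scale-and-fade intro and becomes visible.
# Score updates cap both values at the game state's winning goal when that is positive, then write non-empty results to the text blocks.
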WidgetBlueprint /Game/Athena/HUD/LTM/Respawn/RespawnPostgame_ScoreDisplay
animations [Anim_Intro {
  binding_count=2
  length_seconds=0.45
  object=&self:Anim_Intro
  sequence {bindings [
    Overlay_25 {id=#1, possessed_class=Overlay, tracks []}
    HorizontalBox_0 {
     id=#2
     display_name=VerticalBox_0
     parent_id=#1
     possessed_class=HorizontalBox
     tracks [
      {
       type=MovieScene2DTransformTrack
       sections [{
         type=MovieScene2DTransformSection
         channels [
          Translation.X {type=MovieSceneScriptingFloatChannel}
          Translation.Y {type=MovieSceneScriptingFloatChannel}
          Angle {type=MovieSceneScriptingFloatChannel}
          Scale.X {type=MovieSceneScriptingFloatChannel, keys [[0,1.2],[9,1]]}
          Scale.Y {type=MovieSceneScriptingFloatChannel, keys [[0,1.2],[9,1]]}
          Shear.X {type=MovieSceneScriptingFloatChannel}
          Shear.Y {type=MovieSceneScriptingFloatChannel}
         ]
        }]
      }
      {
       type=MovieSceneFloatTrack
       display_name=Render Opacity
       sections [{type=MovieSceneFloatSection, channels [None {type=MovieSceneScriptingFloatChannel, keys [[0,0],[9,1]]}]}]
      }
     ]
    }
   ]}
 }]
graphs [
 ClampScoreToModeGoalScore {nodes [
   n1 = ClampScoreToModeGoalScore!() -> (then, InScore:Text) {is=title K2Node_FunctionEntry}
   n2 = GetGameStateAthena!() -> (ReturnValue:Fort Game State Athena Object Reference)
   > n3 = Is Valid(InputObject=@n2)
   > n4 = GetScoringModeWinningGoal!(execute, self:Fort Game State Athena Object Reference=@n2) -> (then, ReturnValue:float)
   > n5 = Branch(Condition="float > float"(A=@n4))
   n6 = String To Integer(InString="To String (Text)"(InText=Get InScore() -> (InScore:Text)))
   n7 = String To Integer(InString="To String (Float)"(InDouble=@n4))
   > Return Node!(execute, OutScore:Text="To Text (Integer)"(Value="Min (Integer)"(A=@n7, B=@n6) -> (ReturnValue=0)))
   Return Node(execute=[@n5.else, @n3.Is Not Valid], OutScore=@n1)
  ]}
 EventGraph {nodes [
   n1 = HandleScoreUpdated!() -> (OutputDelegate, then, Scores:Array of Texts) {is=title K2Node_CustomEvent}
   > n2 = ClampScoreToModeGoalScore!(execute, self, InScore:Text="Get (a copy)"(Array:Array of Texts=@n1.Scores) -> (Output:Text)) -> (then, OutScore:Text) {is=title K2Node_CallFunction}
   > n3 = ClampScoreToModeGoalScore(InScore="Get (a copy)"(Array=@n1.Scores, Dimension 1=1))
   > Update!(execute, self, Team1_Score:Text=@n2, Team2_Score:Text=@n3) -> (then)
   Event Construct()
   > SetVisibility(InVisibility=Hidden)
   > Bind Event to Updated(self:Customizable Game Over Score Object Reference, Delegate=@n1) {is=title K2Node_AddDelegate}
   > Delay(Duration=Get AnimDelay() -> (AnimDelay:double))
   > PlayAnimation(InAnimation=Get Anim_Intro() -> (Anim_Intro:Widget Animation Object Reference))
   > SetVisibility(InVisibility=SelfHitTestInvisible)
  ]}
 Update {nodes [
   Update!() -> (then, Team1_Score:Text, Team2_Score:Text) {is=title K2Node_FunctionEntry}
   > n1 = Sequence()
   > Branch(Condition="Not Equal Exactly (Text)"!(A:"Text (by ref)"=Get Team1_Score() -> (Team1_Score:Text), B:"Text (by ref)"=GetEmptyText!() -> (ReturnValue:Text)) -> (ReturnValue:bool))
   > "SetText (Text)"(self=Get CommonTextBlock_Friendly() -> (CommonTextBlock_Friendly:Common Text Object Reference), InText=Get Team1_Score())
   Branch(execute=@n1.then_1, Condition="Not Equal Exactly (Text)"(A=Get Team2_Score() -> (Team2_Score:Text), B=GetEmptyText()))
   > "SetText (Text)"(self=Get CommonTextBlock_Enemy() -> (CommonTextBlock_Enemy:Common Text Object Reference), InText=Get Team2_Score())
  ]}
]
variables [AnimDelay {type=number, default=0.8}]
widgets {
 info {
  inheritedWidgetCount=1
  parentClass=&/Customizable_GameOver_Score.Customizable_GameOver_Score_C
  rootWidgetClass=&Overlay
  widgetCount=12
 }
 widgets [
  Overlay_25 {widgetClassPath=&Overlay, widget=Overlay_25}
  VerticalBox_0 {
   widgetClassPath=&VerticalBox
   parent=Overlay_25
   widget=VerticalBox_0
   slot {class=OverlaySlot, properties {horizontalAlignment=HAlign_Center, verticalAlignment=VAlign_Center}}
  }
  CommonTextBlock_1 {
   widgetClassPath=&CommonTextBlock
   parent=VerticalBox_0
   widget=CommonTextBlock_1
   properties {
    justification=Center
    text=FINAL SCORE
    textTransformPolicy=ToUpper
    font {
     bForceMonospaced=false
     bMaterialIsStencil=false
     fontMaterial=null
     fontObject=&Burbank
     letterSpacing=0
     monospacedWidth=1
     size=32
     skewAmount=0
     typefaceFontName=BurbankRegular-Black
     outlineSettings {
      bApplyOutlineToDropShadows=false
      bMiteredCorners=false
      bSeparateFillAlpha=false
      outlineMaterial=&M-FontShadow
      outlineSize=3
      outlineColor=<rgba 0 0 0 1>
     }
    }
    renderTransform {angle=0, scale=<1 1.3>, shear=<-7 -4>, translation=<0 0>}
    shadowOffset=<0 0>
   }
   slot {class=VerticalBoxSlot, properties {verticalAlignment=VAlign_Center, padding=<ltrb 20 0 20 6>}}
  }
  HorizontalBox_0 {
   widgetClassPath=&HorizontalBox
   parent=VerticalBox_0
   widget=HorizontalBox_0
   properties {renderTransform {angle=0, scale=<1 1>, shear=<-7 -4>, translation=<0 0>}}
   slot {class=VerticalBoxSlot, properties {horizontalAlignment=HAlign_Center, verticalAlignment=VAlign_Center}}
  }
  Overlay_27 {
   widgetClassPath=&Overlay
   parent=HorizontalBox_0
   widget=Overlay_27
   slot {class=HorizontalBoxSlot, properties {horizontalAlignment=HAlign_Center, verticalAlignment=VAlign_Center}}
  }
  ImageBackground_1 {
   widgetClassPath=&Image
   parent=Overlay_27
   widget=ImageBackground_1
   properties {brush {
     bIsDynamicallyLoaded=false
     drawAs=Image
     imageType=NoImage
     mirroring=NoMirror
     resourceName=None
     resourceObject=/Game/Athena/HUD/Materials/Score/MI-TwoTeam_AllyGradient
     tiling=NoTile
     imageSize=<32 32>
     margin=<ltrb 0 0 0 0>
     outlineSettings {
      bUseBrushTransparency=false
      roundingType=HalfHeightRadius
      width=0
      color {colorUseRule=UseColor_Specified, specifiedColor=<rgba 0 0 0 0>}
      cornerRadii=<0 0 0 0>
     }
     tintColor {colorUseRule=UseColor_Specified, specifiedColor=<rgba 1 1 1 1>}
     uVRegion {bIsValid=false, max=<0 0>, min=<0 0>}
    }}
   slot {class=OverlaySlot, properties {horizontalAlignment=HAlign_Fill, verticalAlignment=VAlign_Fill}}
  }
  CommonTextBlock_Friendly {
   widgetClassPath=&CommonTextBlock
   bIsVariable=true
   parent=Overlay_27
   widget=CommonTextBlock_Friendly
   properties {
    justification=Center
    minDesiredWidth=106
    font {
     bForceMonospaced=false
     bMaterialIsStencil=false
     fontMaterial=null
     fontObject=&Burbank
     letterSpacing=0
     monospacedWidth=1
     size=56
     skewAmount=0
     typefaceFontName=Burbank-PlayerPower
     outlineSettings {
      bApplyOutlineToDropShadows=false
      bMiteredCorners=false
      bSeparateFillAlpha=true
      outlineMaterial=/Game/UI/Foundation/Materials/M-FontShadow
      outlineSize=3
      outlineColor=<rgba 0 0.0222 0.3916 0.5>
     }
    }
    shadowColorAndOpacity=<rgba 0 0.0222 0.3916 1>
    shadowOffset=<7 7>
   }
   slot {class=OverlaySlot, properties {horizontalAlignment=HAlign_Center, verticalAlignment=VAlign_Center, padding=<ltrb 4 0 4 0>}}
  }
  CommonTextBlock {
   widgetClassPath=&CommonTextBlock
   parent=HorizontalBox_0
   widget=CommonTextBlock
   properties {
    text=-
    textTransformPolicy=ToUpper
    font {
     bForceMonospaced=false
     bMaterialIsStencil=false
     fontMaterial=null
     fontObject=&Burbank
     letterSpacing=0
     monospacedWidth=1
     size=50
     skewAmount=0
     typefaceFontName=BurbankRegular-Black
     outlineSettings {
      bApplyOutlineToDropShadows=false
      bMiteredCorners=false
      bSeparateFillAlpha=false
      outlineMaterial=/FortUILibrary/Materials/Foundation/M-FontShadow
      outlineSize=3
      outlineColor=<rgba 0 0 0 1>
     }
    }
    renderTransform {angle=-1, scale=<1 1.3>, shear=<-7 -4>, translation=<0 0>}
    shadowOffset=<0 0>
   }
   slot {class=HorizontalBoxSlot, properties {verticalAlignment=VAlign_Center, padding=<ltrb 20 0 20 0>}}
  }
  Overlay_28 {
   widgetClassPath=&Overlay
   parent=HorizontalBox_0
   widget=Overlay_28
   slot {class=HorizontalBoxSlot, properties {horizontalAlignment=HAlign_Center, verticalAlignment=VAlign_Center}}
  }
  ImageBackground_2 {
   widgetClassPath=&Image
   parent=Overlay_28
   widget=ImageBackground_2
   properties {brush {
     bIsDynamicallyLoaded=false
     drawAs=Image
     imageType=NoImage
     mirroring=NoMirror
     resourceName=None
     resourceObject=/Game/Athena/HUD/Materials/Score/MI-TwoTeam_EnemyGradient
     tiling=NoTile
     imageSize=<32 32>
     margin=<ltrb 0 0 0 0>
     outlineSettings {
      bUseBrushTransparency=false
      roundingType=HalfHeightRadius
      width=0
      color {colorUseRule=UseColor_Specified, specifiedColor=<rgba 0 0 0 0>}
      cornerRadii=<0 0 0 0>
     }
     tintColor {colorUseRule=UseColor_Specified, specifiedColor=<rgba 1 1 1 1>}
     uVRegion {bIsValid=false, max=<0 0>, min=<0 0>}
    }}
   slot {class=OverlaySlot, properties {horizontalAlignment=HAlign_Fill, verticalAlignment=VAlign_Fill}}
  }
  CommonTextBlock_Enemy {
   widgetClassPath=&CommonTextBlock
   bIsVariable=true
   parent=Overlay_28
   widget=CommonTextBlock_Enemy
   properties {
    justification=Center
    minDesiredWidth=106
    font {
     bForceMonospaced=false
     bMaterialIsStencil=false
     fontMaterial=null
     fontObject=&Burbank
     letterSpacing=0
     monospacedWidth=1
     size=56
     skewAmount=0
     typefaceFontName=Burbank-PlayerPower
     outlineSettings {
      bApplyOutlineToDropShadows=false
      bMiteredCorners=false
      bSeparateFillAlpha=true
      outlineMaterial=&M-FontShadow
      outlineSize=3
      outlineColor=<rgba 0.0976 0 0.0027 0.5>
     }
    }
    shadowColorAndOpacity=<rgba 0.0976 0 0.0027 1>
    shadowOffset=<7 7>
   }
   slot {class=OverlaySlot, properties {horizontalAlignment=HAlign_Center, verticalAlignment=VAlign_Center, padding=<ltrb 4 0 4 0>}}
  }
  TutorialHighlight {widgetClassPath=&TutorialHighlight}
 ]
}
