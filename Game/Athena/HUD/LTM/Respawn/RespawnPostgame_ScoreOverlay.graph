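# Postgame overlay widget for respawn LTMs that shows win, loss or tie placement with a two-team score display.
# Score updates are capped at the game state's scoring goal through text-to-integer conversion, then passed to the display nested in the game-over score widget.
# The score starts hidden; a loss reveals and animates it immediately, a win waits for the XPScreen placement change, and a tie never triggers it.
# The hero rank HUD element stays hidden until Destruct.
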WidgetBlueprint /Game/Athena/HUD/LTM/Respawn/RespawnPostgame_ScoreOverlay
graphs [
 ClampScoreToModeGoalScore {nodes [
   n1 = ClampScoreToModeGoalScore!() -> (then, InScore:Text) {is=title K2Node_FunctionEntry}
   n2 = GetGameStateAthena!() -> (ReturnValue:Fort Game State Athena Object Reference)
   > n3 = Is Valid(InputObject=@n2)
   > n4 = GetScoringModeWinningGoal!(execute, self:Fort Game State Athena Object Reference=@n2) -> (then, ReturnValue:float)
   > n5 = Branch(Condition="float > float"(A=@n4))
   n6 = String To Integer(InString="To String (Text)"(InText=Get InScore() -> (InScore:Text)))
   n7 = String To Integer(InString="To String (Float)"(InDouble=@n4))
   > Return Node!(execute, OutScore:Text="To Text (Integer)"(Value="Min (Integer)"(A=@n7, B=@n6) -> (ReturnValue=0)))
   Return Node(execute=[@n5.else, @n3.Is Not Valid], OutScore=@n1)
  ]}
 EventGraph {nodes [
   n1 = HandleScoreUpdated!() -> (OutputDelegate, then, Scores:Array of Texts) {is=title K2Node_CustomEvent}
   > n2 = ClampScoreToModeGoalScore!(execute, self, InScore:Text="Get (a copy)"(Array:Array of Texts=@n1.Scores) -> (Output:Text)) -> (then, OutScore:Text) {is=title K2Node_CallFunction}
   > n3 = ClampScoreToModeGoalScore(InScore="Get (a copy)"(Array=@n1.Scores, Dimension 1=1))
   > Update!(
    execute
    self:Respawn Postgame Score Display Object Reference=Get RespawnPostgame_ScoreDisplay() -> (RespawnPostgame_ScoreDisplay:Respawn Postgame Score Display Object Reference)
    Team1_Score:Text=@n2
    Team2_Score:Text=@n3
   ) -> (then)
   n4 = OnWidgetPlacementChanged!() -> (OutputDelegate, then, WidgetPlacement:Name) {is=title K2Node_CustomEvent}
   > Branch(Condition="Equal (Name)"!(A:Name=@n4.WidgetPlacement, B:Name=XPScreen) -> (ReturnValue:bool) {is=title K2Node_CallFunction})
   > ScoreVisibilityAndAnimation()
   Event ShowWinningPlacement() {is=title K2Node_Event}
   n5 = Get PlacementOverlay() -> (PlacementOverlay:Placement Overlay Object Reference)
   > Show Win(self:Placement Overlay Object Reference=@n5)
   > Bind Event to On Post Game Widget Placement Change(
    self:Athena HUDContext Object Reference=Get AthenaHUDContext!() -> (ReturnValue:Athena HUDContext Object Reference) {is=title K2Node_GetBlueprintContextEx}
    Delegate=Create Event()
   ) {is=title K2Node_AddDelegate}
   Event Construct()
   > SetVisibility(self=Get RumbleGameOver_Score() -> (RumbleGameOver_Score:Customizable Game Over Score Object Reference), InVisibility=Hidden)
   > Bind Event to Updated(self:Customizable Game Over Score Object Reference=Get RumbleGameOver_Score(), Delegate=@n1) {is=title K2Node_AddDelegate}
   > SetHUDElementVisibility!(
    execute
    self:Fort HUDContext Object Reference=Get FortHUDContext!() -> (ReturnValue:Fort HUDContext Object Reference) {is=title K2Node_GetBlueprintContextEx}
    HUDElementTags:"Gameplay Tag Container Structure (by ref)"=Get HUDElement_HeroRankTagToHide() -> (HUDElement_HeroRankTagToHide:Gameplay Tag Container Structure)
    bHideElements:bool=true
    ReasonName:"Name (by ref)"=None
   ) -> (then)
   Event ShowLosingPlacement() {is=title K2Node_Event}
   > Show Loss(self:Placement Overlay Object Reference=@n5)
   > ScoreVisibilityAndAnimation()
   Event ShowTiePlacement() {is=title K2Node_Event}
   > Show Tie(self:Placement Overlay Object Reference=@n5)
   Event Destruct()
   > Unbind Event from On Post Game Widget Placement Change(self:Athena HUDContext Object Reference=Get AthenaHUDContext(), Delegate=Create Event()) {is=title K2Node_RemoveDelegate}
   > SetHUDElementVisibility(self=Get FortHUDContext(), HUDElementTags=Get HUDElement_HeroRankTagToHide(), bHideElements=false, ReasonName=None)
  ]}
 ScoreVisibilityAndAnimation {nodes [
   ScoreVisibilityAndAnimation() {is=title K2Node_FunctionEntry}
   n1 = Get RespawnPostgame_ScoreDisplay()
   > PlayAnimation(self=@n1, InAnimation=Get Intro!(self:Customizable Game Over Score Object Reference=@n1) -> (Intro))
   > Unbind Event from On Post Game Widget Placement Change(self=Get AthenaHUDContext(), Delegate=Create Event())
   > SetVisibility(self=Get RumbleGameOver_Score(), InVisibility=HitTestInvisible)
  ]}
]
variables [HUDElement_HeroRankTagToHide {type=GameplayTagContainer, default {gameplayTags [{tagName=CherrySmokeHeroRank.HUD.HeroRank}]}}]
widgets {
 info {
  inheritedWidgetCount=1
  namedSlotCount=1
  parentClass=/Script/FortniteUI.PostGamePlacementOverlay
  rootWidgetClass=&Overlay
  widgetCount=6
 }
 widgets [
  Overlay_34 {widgetClassPath=&Overlay, widget=Overlay_34}
  VerticalBox_39 {
   widgetClassPath=&VerticalBox
   parent=Overlay_34
   widget=VerticalBox_39
   slot {class=OverlaySlot, properties {horizontalAlignment=HAlign_Center}}
  }
  PlacementOverlay {
   widgetClassPath=/Game/Athena/HUD/EndGame/PlacementOverlay
   bIsVariable=true
   parent=VerticalBox_39
   widget=PlacementOverlay
   properties {renderTransform {angle=0, scale=<0.9 0.9>, shear=<0 0>, translation=<0 0>}}
   slot {class=VerticalBoxSlot, properties {horizontalAlignment=HAlign_Center, verticalAlignment=VAlign_Top}}
  }
  RumbleGameOver_Score {
   widgetClassPath=&/Customizable_GameOver_Score
   bIsVariable=true
   parent=Overlay_34
   widget=RumbleGameOver_Score
   properties {elementTag {tagName=HUD.Score.TeamScore}, renderTransform {angle=0, scale=<0.8 0.8>, shear=<0 0>, translation=<0 225>}}
   slot {class=OverlaySlot, properties {horizontalAlignment=HAlign_Center}}
  }
  RespawnPostgame_ScoreDisplay {
   widgetClassPath=&/RespawnPostgame_ScoreDisplay
   bIsVariable=true
   namedSlotHost=RumbleGameOver_Score
   widget=RespawnPostgame_ScoreDisplay
  }
  TutorialHighlight {widgetClassPath=&TutorialHighlight}
 ]
}
